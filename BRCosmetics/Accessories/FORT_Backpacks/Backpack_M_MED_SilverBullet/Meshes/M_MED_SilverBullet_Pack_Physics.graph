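# Physics asset for the SilverBullet medium male backpack cosmetic, setting up collision and joints for its dangling keychain.
# A large capsule on the root bone covers the pack, and three smaller capsules follow the dyn_keychain bone chain, the last one fattest.
# Constraints link the chain in sequence from root to tip with limited swing and locked twist, with the first joint tighter than the rest.
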PhysicsAsset /BRCosmetics/Accessories/FORT_Backpacks/Backpack_M_MED_SilverBullet/Meshes/M_MED_SilverBullet_Pack_Physics
bodies {
 dyn_keychain_01 [{shapeType=Capsule, length=2.005, radius=1.005, shapeName=dyn_keychain_01_capsule, center=<1.3649 0 -0.0313>, rotation=<rot 90 0 0>}]
 dyn_keychain_02 [{shapeType=Capsule, length=2.005, radius=1.005, shapeName=dyn_keychain_01_capsule, center=<1.7187 0 -0.0312>, rotation=<rot 90 0 0>}]
 dyn_keychain_03 [{shapeType=Capsule, length=2, radius=2.5, shapeName=dyn_keychain_01_capsule, center=<4.046 0 0>, rotation=<rot 90 0 0>}]
 root [{shapeType=Capsule, length=13.5, radius=8, shapeName=root_capsule, center=<0 -11.3017 -12.6316>, rotation=<rot -0.0002 -90 90>}]
}
constraints [
 {bone1Name=dyn_keychain_03, bone2Name=dyn_keychain_02, swing1LimitDegrees=35, swing2LimitDegrees=35, twistLimitDegrees=45, twistMotion=Locked}
 {bone1Name=dyn_keychain_01, bone2Name=root, swing1LimitDegrees=25, swing2LimitDegrees=25, twistLimitDegrees=45, twistMotion=Locked}
 {bone1Name=dyn_keychain_02, bone2Name=dyn_keychain_01, swing1LimitDegrees=35, swing2LimitDegrees=35, twistLimitDegrees=45, twistMotion=Locked}
]
